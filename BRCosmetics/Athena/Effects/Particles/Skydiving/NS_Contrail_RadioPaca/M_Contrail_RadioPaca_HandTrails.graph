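# Translucent, unlit, two-sided Niagara mesh-particle material for the hand trails of the RadioPaca skydiving contrail.
# Vertex colors, scrolling noise and audio-reactivity dynamic parameters build masks, with an isUnderline parameter apparently selecting a separate strip look.
# The color mask, scaled by DiveColorMult, indexes a curve-atlas gradient boosted to HDR emissive.
# Opacity combines a hard vertex-green threshold, vertex alpha and a custom-depth occlusion test, which mobile skips.
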Material /BRCosmetics/Athena/Effects/Particles/Skydiving/NS_Contrail_RadioPaca/M_Contrail_RadioPaca_HandTrails
graph {
 expressions {
  MaterialExpressionAdd_0 {
   class=MaterialExpressionAdd
   inputs [
    A {from=MaterialExpressionMultiply_2}
    B {from=MaterialExpressionFrac_1}
   ]
  }
  MaterialExpressionAdd_1 {
   class=MaterialExpressionAdd
   inputs [
    A {from=MaterialExpressionVertexColor_6, output=B}
    B {from=MaterialExpressionMultiply_1}
   ]
  }
  MaterialExpressionAdd_16 {
   class=MaterialExpressionAdd
   inputs [
    A {from=MaterialExpressionMultiply_31}
    B {from=MaterialExpressionFrac_8}
   ]
  }
  MaterialExpressionAdd_2 {
   class=MaterialExpressionAdd
   inputs [
    A {from=MaterialExpressionClamp_0}
    B {from=MaterialExpressionLinearInterpolate_4}
   ]
  }
  MaterialExpressionAdd_22 {
   class=MaterialExpressionAdd
   inputs [
    A {from=MaterialExpressionMultiply_53}
    B {from=MaterialExpressionLinearInterpolate_18}
   ]
   properties {constB=0}
  }
  MaterialExpressionAdd_23 {
   class=MaterialExpressionAdd
   inputs [
    A {from=MaterialExpressionMultiply_58}
    B {from=MaterialExpressionFrac_10}
   ]
  }
  MaterialExpressionAdd_24 {
   class=MaterialExpressionAdd
   inputs [
    A {from=MaterialExpressionMultiply_60}
    B {from=MaterialExpressionMultiply_61}
   ]
  }
  MaterialExpressionAdd_25 {
   class=MaterialExpressionAdd
   inputs [
    A {from=MaterialExpressionMultiply_70}
    B {from=MaterialExpressionFrac_11}
   ]
  }
  MaterialExpressionAdd_26 {
   class=MaterialExpressionAdd
   inputs [
    A {from=MaterialExpressionAdd_24}
    B {from=MaterialExpressionMultiply_71}
   ]
  }
  MaterialExpressionAdd_3 {
   class=MaterialExpressionAdd
   inputs [
    A {from=MaterialExpressionMultiply_12}
    B {from=MaterialExpressionMultiply_10}
   ]
  }
  MaterialExpressionAdd_4 {
   class=MaterialExpressionAdd
   inputs [
    A {from=MaterialExpressionPixelDepth_0, output=R}
    B {from=MaterialExpressionScalarParameter_33}
   ]
   properties {constB=-1}
  }
  MaterialExpressionAdd_5 {
   class=MaterialExpressionAdd
   inputs [
    A {from=MaterialExpressionLinearInterpolate_3}
    B {from=MaterialExpressionLinearInterpolate_0}
   ]
  }
  MaterialExpressionAdd_6 {
   class=MaterialExpressionAdd
   inputs [
    A {from=MaterialExpressionMultiply_0}
    B {from=MaterialExpressionLinearInterpolate_28}
   ]
   properties {constB=0}
  }
  MaterialExpressionClamp_0 {
   class=MaterialExpressionClamp
   inputs [
    None {from=MaterialExpressionLinearInterpolate_20, output=""}
    Min {from="", output=""}
    Max {from="", output=""}
   ]
   properties {maxDefault=1.1, minDefault=0.05}
  }
  MaterialExpressionComponentMask_0 {class=MaterialExpressionComponentMask, inputs [None {from=MaterialExpressionSceneTexture_0, output=Color}], properties {r=true}}
  MaterialExpressionConstant_10 {class=MaterialExpressionConstant, properties {r=1}}
  MaterialExpressionConstant_11 {class=MaterialExpressionConstant, properties {r=1}}
  MaterialExpressionConstant_12 {class=MaterialExpressionConstant, properties {r=1}}
  MaterialExpressionConstant_2 {class=MaterialExpressionConstant, properties {r=0.65}}
  MaterialExpressionConstant_3 {class=MaterialExpressionConstant}
  MaterialExpressionConstant_4 {class=MaterialExpressionConstant, properties {r=1}}
  MaterialExpressionConstant_8 {class=MaterialExpressionConstant}
  MaterialExpressionConstant_9 {class=MaterialExpressionConstant}
  MaterialExpressionConvert_0 {
   class=MaterialExpressionConvert
   inputs [
    None {from=MaterialExpressionAdd_0}
    None {from=MaterialExpressionFrac_3}
   ]
   properties {
    nodeName=Make Float2
    convertInputs [
     {
      type=Scalar
      expressionInput {
       expression=&self:MaterialExpressionAdd_0
       inputName=None
       mask=0
       maskA=0
       maskB=0
       maskG=0
       maskR=0
       outputIndex=0
      }
     }
     {
      type=Scalar
      expressionInput {
       expression=&self:MaterialExpressionFrac_3
       inputName=None
       mask=0
       maskA=0
       maskB=0
       maskG=0
       maskR=0
       outputIndex=0
      }
     }
    ]
    convertOutputs [{type=Vector2}]
   }
  }
  MaterialExpressionConvert_3 {
   class=MaterialExpressionConvert
   inputs [
    None {from=MaterialExpressionAdd_6}
    None {from=MaterialExpressionAdd_16}
   ]
   properties {
    nodeName=Make Float2
    convertInputs [
     {
      type=Scalar
      expressionInput {
       expression=&self:MaterialExpressionAdd_6
       inputName=None
       mask=0
       maskA=0
       maskB=0
       maskG=0
       maskR=0
       outputIndex=0
      }
     }
     {
      type=Scalar
      expressionInput {
       expression=&self:MaterialExpressionAdd_16
       inputName=None
       mask=0
       maskA=0
       maskB=0
       maskG=0
       maskR=0
       outputIndex=0
      }
     }
    ]
    convertOutputs [{type=Vector2}]
   }
  }
  MaterialExpressionConvert_4 {
   class=MaterialExpressionConvert
   inputs [
    None {from=MaterialExpressionAdd_25}
    None {from=""}
   ]
   properties {
    nodeName=Make Float2
    convertInputs [
     {
      type=Scalar
      expressionInput {
       expression=&self:MaterialExpressionAdd_25
       inputName=None
       mask=0
       maskA=0
       maskB=0
       maskG=0
       maskR=0
       outputIndex=0
      }
     }
     {type=Scalar, expressionInput {expression=null, inputName=None, mask=0, maskA=0, maskB=0, maskG=0, maskR=0, outputIndex=0}}
    ]
    convertOutputs [{type=Vector2}]
   }
  }
  MaterialExpressionConvert_5 {
   class=MaterialExpressionConvert
   inputs [
    None {from=MaterialExpressionAdd_22}
    None {from=MaterialExpressionAdd_23}
   ]
   properties {
    nodeName=Make Float2
    convertInputs [
     {
      type=Scalar
      expressionInput {
       expression=&self:MaterialExpressionAdd_22
       inputName=None
       mask=0
       maskA=0
       maskB=0
       maskG=0
       maskR=0
       outputIndex=0
      }
     }
     {
      type=Scalar
      expressionInput {
       expression=&self:MaterialExpressionAdd_23
       inputName=None
       mask=0
       maskA=0
       maskB=0
       maskG=0
       maskR=0
       outputIndex=0
      }
     }
    ]
    convertOutputs [{type=Vector2}]
   }
  }
  MaterialExpressionCurveAtlasRowParameter_0 {
   class=MaterialExpressionCurveAtlasRowParameter
   inputs [CurveTime {from=MaterialExpressionMultiply_66, output=""}]
   properties {
    atlas=/Game/Athena/LinearColorCurves/CA_Nebula_Contrail
    curve=/Game/Athena/LinearColorCurves/CC_Contrail_RadioPaca_Color
    defaultValue=5
    parameterName=ColorCurve
   }
  }
  MaterialExpressionCurveAtlasRowParameter_1 {
   class=MaterialExpressionCurveAtlasRowParameter
   inputs [CurveTime {from=MaterialExpressionVertexColor_0, output=R}]
   properties {
    atlas=&CA_Nebula_Contrail
    curve=/Game/Athena/LinearColorCurves/CC_Contrail_RadioPaca_Red
    defaultValue=6
    parameterName=Color_RedAdd
   }
  }
  MaterialExpressionDivide_0 {
   class=MaterialExpressionDivide
   inputs [
    A {from=MaterialExpressionFrac_0}
    B {from=""}
   ]
  }
  MaterialExpressionDynamicParameter_0 {class=MaterialExpressionDynamicParameter, inputs [], properties {paramNames [AudioReactivity, HDRMult, DiveColorMult, AudioReactivity2]}}
  MaterialExpressionDynamicParameter_10 {class=MaterialExpressionDynamicParameter, inputs [], properties {paramNames [AudioReactivity, HDRMult, DiveColorMult, AudioReactivity2]}}
  MaterialExpressionDynamicParameter_11 {class=MaterialExpressionDynamicParameter, inputs [], properties {paramNames [AudioReactivity, HDRMult, DiveColorMult, AudioReactivity2]}}
  MaterialExpressionDynamicParameter_12 {class=MaterialExpressionDynamicParameter, inputs [], properties {paramNames [AudioReactivity, HDRMult, DiveColorMult, AudioReactivity2]}}
  MaterialExpressionDynamicParameter_13 {class=MaterialExpressionDynamicParameter, inputs [], properties {paramNames [AudioReactivity, HDRMult, DiveColorMult, AudioReactivity2]}}
  MaterialExpressionDynamicParameter_14 {class=MaterialExpressionDynamicParameter, inputs [], properties {paramNames [AudioReactivity, HDRMult, DiveColorMult, AudioReactivity2]}}
  MaterialExpressionDynamicParameter_3 {class=MaterialExpressionDynamicParameter, inputs [], properties {paramNames [AudioReactivity, HDRMult, DiveColorMult, AudioReactivity2]}}
  MaterialExpressionDynamicParameter_4 {class=MaterialExpressionDynamicParameter, inputs [], properties {paramNames [AudioReactivity, HDRMult, DiveColorMult, AudioReactivity2]}}
  MaterialExpressionDynamicParameter_5 {class=MaterialExpressionDynamicParameter, inputs [], properties {paramNames [AudioReactivity, HDRMult, DiveColorMult, AudioReactivity2]}}
  MaterialExpressionDynamicParameter_6 {class=MaterialExpressionDynamicParameter, inputs [], properties {paramNames [AudioReactivity, HDRMult, DiveColorMult, AudioReactivity2]}}
  MaterialExpressionDynamicParameter_7 {class=MaterialExpressionDynamicParameter, inputs [], properties {paramNames [AudioReactivity, HDRMult, DiveColorMult, AudioReactivity2]}}
  MaterialExpressionDynamicParameter_8 {class=MaterialExpressionDynamicParameter, inputs [], properties {paramNames [AudioReactivity, HDRMult, DiveColorMult, AudioReactivity2]}}
  MaterialExpressionFrac_0 {class=MaterialExpressionFrac, inputs [None {from=MaterialExpressionAdd_1, output=""}], properties {}}
  MaterialExpressionFrac_1 {class=MaterialExpressionFrac, inputs [None {from=MaterialExpressionTime_0, output=""}], properties {}}
  MaterialExpressionFrac_10 {class=MaterialExpressionFrac, inputs [None {from=MaterialExpressionMultiply_59, output=""}], properties {}}
  MaterialExpressionFrac_11 {class=MaterialExpressionFrac, inputs [None {from=MaterialExpressionTime_11, output=""}], properties {}}
  MaterialExpressionFrac_2 {class=MaterialExpressionFrac, inputs [None {from=MaterialExpressionTime_2, output=""}], properties {}}
  MaterialExpressionFrac_3 {class=MaterialExpressionFrac, inputs [None {from=MaterialExpressionMultiply_6, output=""}], properties {}}
  MaterialExpressionFrac_8 {class=MaterialExpressionFrac, inputs [None {from=MaterialExpressionMultiply_32, output=""}], properties {}}
  MaterialExpressionFrac_9 {class=MaterialExpressionFrac, inputs [None {from=MaterialExpressionTime_9, output=""}], properties {}}
  MaterialExpressionIf_0 {
   class=MaterialExpressionIf
   inputs [
    A {from=MaterialExpressionVertexColor_23, output=G}
    B {from=MaterialExpressionNamedRerouteUsage_20, output=""}
    "A > B" {from=MaterialExpressionConstant_3, output=""}
    "A == B" {from="", output=""}
    "A < B" {from=MaterialExpressionConstant_4, output=""}
   ]
   properties {constB=0.5}
  }
  MaterialExpressionIf_1 {
   class=MaterialExpressionIf
   inputs [
    A {from=MaterialExpressionAdd_4, output=""}
    B {from=MaterialExpressionComponentMask_0, output=""}
    "A > B" {from=MaterialExpressionConstant_8, output=""}
    "A == B" {from=MaterialExpressionConstant_8, output=""}
    "A < B" {from=MaterialExpressionConstant_11, output=""}
   ]
   properties {}
  }
  MaterialExpressionLinearInterpolate_0 {
   class=MaterialExpressionLinearInterpolate
   inputs [
    A {from=""}
    B {from=""}
    Alpha {from=MaterialExpressionTextureSample_0, output=R}
   ]
   properties {constB=0.75}
  }
  MaterialExpressionLinearInterpolate_1 {
   class=MaterialExpressionLinearInterpolate
   inputs [
    A {from=MaterialExpressionScalarParameter_2}
    B {from=MaterialExpressionLinearInterpolate_35}
    Alpha {from=MaterialExpressionAdd_3}
   ]
   properties {constA=0.1}
  }
  MaterialExpressionLinearInterpolate_10 {
   class=MaterialExpressionLinearInterpolate
   inputs [
    A {from=""}
    B {from=""}
    Alpha {from=MaterialExpressionNamedRerouteUsage_4}
   ]
   properties {constA=0.25}
  }
  MaterialExpressionLinearInterpolate_12 {
   class=MaterialExpressionLinearInterpolate
   inputs [
    A {from=""}
    B {from=MaterialExpressionLinearInterpolate_34}
    Alpha {from=MaterialExpressionNamedRerouteUsage_5}
   ]
   properties {constB=0.25}
  }
  MaterialExpressionLinearInterpolate_13 {
   class=MaterialExpressionLinearInterpolate
   inputs [
    A {from=""}
    B {from=""}
    Alpha {from=MaterialExpressionOneMinus_2}
   ]
   properties {constA=-0.5}
  }
  MaterialExpressionLinearInterpolate_18 {
   class=MaterialExpressionLinearInterpolate
   inputs [
    A {from=MaterialExpressionMultiply_15}
    B {from=MaterialExpressionMultiply_41}
    Alpha {from=MaterialExpressionDynamicParameter_10, output=AudioReactivity}
   ]
  }
  MaterialExpressionLinearInterpolate_2 {
   class=MaterialExpressionLinearInterpolate
   inputs [
    A {from=MaterialExpressionMultiply_7}
    B {from=MaterialExpressionVertexColor_1, output=A}
    Alpha {from=MaterialExpressionNamedRerouteUsage_1}
   ]
  }
  MaterialExpressionLinearInterpolate_20 {
   class=MaterialExpressionLinearInterpolate
   inputs [
    A {from=MaterialExpressionLinearInterpolate_26}
    B {from=MaterialExpressionLinearInterpolate_32}
    Alpha {from=MaterialExpressionMultiply_27}
   ]
  }
  MaterialExpressionLinearInterpolate_25 {
   class=MaterialExpressionLinearInterpolate
   inputs [
    A {from=MaterialExpressionScalarParameter_39}
    B {from=MaterialExpressionLinearInterpolate_30}
    Alpha {from=MaterialExpressionVertexColor_29, output=R}
   ]
   properties {constA=1, constB=3}
  }
  MaterialExpressionLinearInterpolate_26 {
   class=MaterialExpressionLinearInterpolate
   inputs [
    A {from=""}
    B {from=MaterialExpressionLinearInterpolate_31}
    Alpha {from=MaterialExpressionVertexColor_30, output=R}
   ]
   properties {constA=-0.2, constB=0.5}
  }
  MaterialExpressionLinearInterpolate_27 {
   class=MaterialExpressionLinearInterpolate
   inputs [
    A {from=""}
    B {from=""}
    Alpha {from=MaterialExpressionDynamicParameter_11, output=AudioReactivity}
   ]
   properties {constB=0.25}
  }
  MaterialExpressionLinearInterpolate_28 {
   class=MaterialExpressionLinearInterpolate
   inputs [
    A {from=MaterialExpressionMultiply_54}
    B {from=MaterialExpressionMultiply_62}
    Alpha {from=MaterialExpressionDynamicParameter_13, output=AudioReactivity}
   ]
  }
  MaterialExpressionLinearInterpolate_3 {
   class=MaterialExpressionLinearInterpolate
   inputs [
    A {from=MaterialExpressionLinearInterpolate_1}
    B {from=MaterialExpressionNamedRerouteUsage_22}
    Alpha {from=MaterialExpressionNamedRerouteUsage_2}
   ]
  }
  MaterialExpressionLinearInterpolate_30 {
   class=MaterialExpressionLinearInterpolate
   inputs [
    A {from=""}
    B {from=""}
    Alpha {from=MaterialExpressionDynamicParameter_3, output=AudioReactivity}
   ]
   properties {constA=2, constB=5}
  }
  MaterialExpressionLinearInterpolate_31 {
   class=MaterialExpressionLinearInterpolate
   inputs [
    A {from=""}
    B {from=""}
    Alpha {from=MaterialExpressionDynamicParameter_4, output=AudioReactivity}
   ]
   properties {constA=0.2, constB=0.5}
  }
  MaterialExpressionLinearInterpolate_32 {
   class=MaterialExpressionLinearInterpolate
   inputs [
    A {from=""}
    B {from=""}
    Alpha {from=MaterialExpressionDynamicParameter_5, output=AudioReactivity}
   ]
   properties {constA=0.25}
  }
  MaterialExpressionLinearInterpolate_33 {
   class=MaterialExpressionLinearInterpolate
   inputs [
    A {from=""}
    B {from=""}
    Alpha {from=MaterialExpressionDynamicParameter_6, output=AudioReactivity}
   ]
   properties {constA=1, constB=1.25}
  }
  MaterialExpressionLinearInterpolate_34 {
   class=MaterialExpressionLinearInterpolate
   inputs [
    A {from=""}
    B {from=""}
    Alpha {from=MaterialExpressionDynamicParameter_7, output=AudioReactivity}
   ]
   properties {constA=0.1, constB=0.25}
  }
  MaterialExpressionLinearInterpolate_35 {
   class=MaterialExpressionLinearInterpolate
   inputs [
    A {from=""}
    B {from=MaterialExpressionScalarParameter_3}
    Alpha {from=MaterialExpressionDynamicParameter_8, output=AudioReactivity}
   ]
   properties {constA=0.5}
  }
  MaterialExpressionLinearInterpolate_37 {
   class=MaterialExpressionLinearInterpolate
   inputs [
    A {from=""}
    B {from=""}
    Alpha {from=MaterialExpressionDynamicParameter_12, output=AudioReactivity}
   ]
   properties {constB=0.5}
  }
  MaterialExpressionLinearInterpolate_4 {
   class=MaterialExpressionLinearInterpolate
   inputs [
    A {from=MaterialExpressionMultiply_5}
    B {from=""}
    Alpha {from=MaterialExpressionDynamicParameter_0, output=AudioReactivity2}
   ]
   properties {constB=0}
  }
  MaterialExpressionLinearInterpolate_40 {
   class=MaterialExpressionLinearInterpolate
   inputs [
    A {from=""}
    B {from=""}
    Alpha {from=MaterialExpressionOneMinus_3}
   ]
   properties {constA=0.15}
  }
  MaterialExpressionLinearInterpolate_41 {
   class=MaterialExpressionLinearInterpolate
   inputs [
    A {from=""}
    B {from=""}
    Alpha {from=MaterialExpressionNamedRerouteUsage_21}
   ]
   properties {constA=0.1}
  }
  MaterialExpressionLinearInterpolate_42 {
   class=MaterialExpressionLinearInterpolate
   inputs [
    A {from=""}
    B {from=""}
    Alpha {from=MaterialExpressionPower_7}
   ]
   properties {constA=0.35}
  }
  MaterialExpressionLinearInterpolate_43 {
   class=MaterialExpressionLinearInterpolate
   inputs [
    A {from=""}
    B {from=""}
    Alpha {from=MaterialExpressionVertexColor_39, output=R}
   ]
   properties {constA=0.5}
  }
  MaterialExpressionLinearInterpolate_8 {
   class=MaterialExpressionLinearInterpolate
   inputs [
    A {from=MaterialExpressionScalarParameter_10}
    B {from=MaterialExpressionScalarParameter_11}
    Alpha {from=MaterialExpressionNamedRerouteUsage_3}
   ]
  }
  MaterialExpressionLinearInterpolate_9 {
   class=MaterialExpressionLinearInterpolate
   inputs [
    A {from=MaterialExpressionLinearInterpolate_8}
    B {from=""}
    Alpha {from=MaterialExpressionSine_0}
   ]
  }
  MaterialExpressionMaterialFunctionCall_0 {
   class=MaterialExpressionMaterialFunctionCall
   inputs [Emissive {from=MaterialExpressionMaterialFunctionCall_3, output=RemappedRange}]
   properties {materialFunction=&MF_FX_EmissiveAdjustments}
  }
  MaterialExpressionMaterialFunctionCall_3 {
   class=MaterialExpressionMaterialFunctionCall
   inputs [
    LDR Input {from=MaterialExpressionCurveAtlasRowParameter_0}
    HDR Dynamic Range {from=""}
    HDR Tint {from=""}
    HDR Intensity Multiplier {from=MaterialExpressionMultiply_63}
    Derive HDR Power {from=""}
    Desaturate Luminance {from=""}
    Derive Luminance from Linear {from=""}
   ]
   properties {materialFunction=&DeriveHDRfromLDR}
  }
  MaterialExpressionMaterialFunctionCall_4 {
   class=MaterialExpressionMaterialFunctionCall
   inputs [
    Input {from=MaterialExpressionVertexColor_6, output=B}
    min {from=MaterialExpressionConstant_9}
    max {from=MaterialExpressionConstant_10}
    seed {from=""}
   ]
   properties {materialFunction=/InterchangeAssets/Functions/MX_RandomFloat}
  }
  MaterialExpressionMultiply_0 {
   class=MaterialExpressionMultiply
   inputs [
    A {from=MaterialExpressionVertexColor_18, output=R}
    B {from=MaterialExpressionScalarParameter_14}
   ]
   properties {constB=4}
  }
  MaterialExpressionMultiply_1 {
   class=MaterialExpressionMultiply
   inputs [
    A {from=MaterialExpressionMaterialFunctionCall_4, output=Result}
    B {from=MaterialExpressionTime_1}
   ]
   properties {constB=0.5}
  }
  MaterialExpressionMultiply_10 {
   class=MaterialExpressionMultiply
   inputs [
    A {from=MaterialExpressionMultiply_11}
    B {from=MaterialExpressionLinearInterpolate_9}
   ]
  }
  MaterialExpressionMultiply_11 {
   class=MaterialExpressionMultiply
   inputs [
    A {from=MaterialExpressionMultiply_4}
    B {from=MaterialExpressionLinearInterpolate_10}
   ]
  }
  MaterialExpressionMultiply_12 {
   class=MaterialExpressionMultiply
   inputs [
    A {from=MaterialExpressionLinearInterpolate_12}
    B {from=MaterialExpressionSaturate_0}
   ]
  }
  MaterialExpressionMultiply_15 {
   class=MaterialExpressionMultiply
   inputs [
    A {from=MaterialExpressionFrac_2}
    B {from=""}
   ]
   properties {constB=0}
  }
  MaterialExpressionMultiply_2 {
   class=MaterialExpressionMultiply
   inputs [
    A {from=MaterialExpressionVertexColor_2, output=R}
    B {from=MaterialExpressionScalarParameter_1}
   ]
   properties {constB=8}
  }
  MaterialExpressionMultiply_27 {
   class=MaterialExpressionMultiply
   inputs [
    A {from=MaterialExpressionAdd_26}
    B {from=MaterialExpressionLinearInterpolate_25}
   ]
   properties {constB=2}
  }
  MaterialExpressionMultiply_3 {
   class=MaterialExpressionMultiply
   inputs [
    A {from=MaterialExpressionLinearInterpolate_2}
    B {from=MaterialExpressionShadingPathSwitch_0}
   ]
   properties {constA=1}
  }
  MaterialExpressionMultiply_31 {
   class=MaterialExpressionMultiply
   inputs [
    A {from=MaterialExpressionVertexColor_28, output=R}
    B {from=MaterialExpressionScalarParameter_28}
   ]
   properties {constB=4}
  }
  MaterialExpressionMultiply_32 {
   class=MaterialExpressionMultiply
   inputs [
    A {from=MaterialExpressionTime_3}
    B {from=MaterialExpressionScalarParameter_29}
   ]
  }
  MaterialExpressionMultiply_4 {
   class=MaterialExpressionMultiply
   inputs [
    A {from=MaterialExpressionOneMinus_2}
    B {from=MaterialExpressionScalarParameter_5}
   ]
   properties {constB=10}
  }
  MaterialExpressionMultiply_41 {
   class=MaterialExpressionMultiply
   inputs [
    A {from=MaterialExpressionFrac_2}
    B {from=""}
   ]
   properties {constB=0}
  }
  MaterialExpressionMultiply_5 {
   class=MaterialExpressionMultiply
   inputs [
    A {from=MaterialExpressionTextureSample_3, output=B}
    B {from=MaterialExpressionConstant_2}
   ]
   properties {constB=0.45}
  }
  MaterialExpressionMultiply_53 {
   class=MaterialExpressionMultiply
   inputs [
    A {from=MaterialExpressionVertexColor_31, output=R}
    B {from=MaterialExpressionScalarParameter_34}
   ]
   properties {constB=4}
  }
  MaterialExpressionMultiply_54 {
   class=MaterialExpressionMultiply
   inputs [
    A {from=MaterialExpressionFrac_9}
    B {from=""}
   ]
   properties {constB=0}
  }
  MaterialExpressionMultiply_58 {
   class=MaterialExpressionMultiply
   inputs [
    A {from=MaterialExpressionVertexColor_32, output=R}
    B {from=MaterialExpressionScalarParameter_36}
   ]
   properties {constB=4}
  }
  MaterialExpressionMultiply_59 {
   class=MaterialExpressionMultiply
   inputs [
    A {from=MaterialExpressionTime_10}
    B {from=MaterialExpressionScalarParameter_37}
   ]
  }
  MaterialExpressionMultiply_6 {
   class=MaterialExpressionMultiply
   inputs [
    A {from=MaterialExpressionTime_4}
    B {from=MaterialExpressionScalarParameter_6}
   ]
  }
  MaterialExpressionMultiply_60 {
   class=MaterialExpressionMultiply
   inputs [
    A {from=MaterialExpressionTextureSample_1, output=R}
    B {from=MaterialExpressionScalarParameter_22}
   ]
   properties {constB=0.65}
  }
  MaterialExpressionMultiply_61 {
   class=MaterialExpressionMultiply
   inputs [
    A {from=MaterialExpressionTextureSample_5, output=R}
    B {from=MaterialExpressionScalarParameter_30}
   ]
   properties {constB=0.35}
  }
  MaterialExpressionMultiply_62 {
   class=MaterialExpressionMultiply
   inputs [
    A {from=MaterialExpressionFrac_9}
    B {from=""}
   ]
   properties {constB=0}
  }
  MaterialExpressionMultiply_63 {
   class=MaterialExpressionMultiply
   inputs [
    A {from=MaterialExpressionLinearInterpolate_33}
    B {from=MaterialExpressionDynamicParameter_6, output=HDRMult}
   ]
  }
  MaterialExpressionMultiply_65 {
   class=MaterialExpressionMultiply
   inputs [
    A {from=MaterialExpressionFrac_10}
    B {from=MaterialExpressionLinearInterpolate_27}
   ]
  }
  MaterialExpressionMultiply_66 {
   class=MaterialExpressionMultiply
   inputs [
    A {from=MaterialExpressionAdd_5}
    B {from=MaterialExpressionDynamicParameter_14, output=DiveColorMult}
   ]
  }
  MaterialExpressionMultiply_67 {
   class=MaterialExpressionMultiply
   inputs [
    A {from=MaterialExpressionFrac_8}
    B {from=MaterialExpressionLinearInterpolate_37}
   ]
  }
  MaterialExpressionMultiply_7 {
   class=MaterialExpressionMultiply
   inputs [
    A {from=MaterialExpressionIf_0}
    B {from=MaterialExpressionVertexColor_1, output=A}
   ]
  }
  MaterialExpressionMultiply_70 {
   class=MaterialExpressionMultiply
   inputs [
    A {from=MaterialExpressionVertexColor_35, output=R}
    B {from=MaterialExpressionScalarParameter_32}
   ]
   properties {constB=8}
  }
  MaterialExpressionMultiply_71 {
   class=MaterialExpressionMultiply
   inputs [
    A {from=MaterialExpressionTextureSample_2, output=B}
    B {from=MaterialExpressionScalarParameter_31}
   ]
   properties {constB=0.45}
  }
  MaterialExpressionMultiply_73 {
   class=MaterialExpressionMultiply
   inputs [
    A {from=MaterialExpressionPower_6}
    B {from=MaterialExpressionLinearInterpolate_43}
   ]
  }
  MaterialExpressionMultiply_74 {
   class=MaterialExpressionMultiply
   inputs [
    A {from=MaterialExpressionMultiply_75}
    B {from=MaterialExpressionLinearInterpolate_42}
   ]
  }
  MaterialExpressionMultiply_75 {
   class=MaterialExpressionMultiply
   inputs [
    A {from=MaterialExpressionLinearInterpolate_40}
    B {from=MaterialExpressionMultiply_73}
   ]
  }
  MaterialExpressionNamedRerouteDeclaration_0 {
   class=MaterialExpressionNamedRerouteDeclaration
   inputs [None {from=MaterialExpressionMaterialFunctionCall_0, output=Result}]
   properties {name=ToEmissive, nodeColor=<rgba 0.9922 0 1 1>}
  }
  MaterialExpressionNamedRerouteDeclaration_1 {
   class=MaterialExpressionNamedRerouteDeclaration
   inputs [None {from=MaterialExpressionScalarParameter_35}]
   properties {name=SeperatedStrip, nodeColor=<rgba 0.8516 1 0 1>}
  }
  MaterialExpressionNamedRerouteDeclaration_12 {
   class=MaterialExpressionNamedRerouteDeclaration
   inputs [None {from=MaterialExpressionAdd_2}]
   properties {name=Green, nodeColor=<rgba 0.2656 1 0 1>}
  }
  MaterialExpressionNamedRerouteDeclaration_14 {
   class=MaterialExpressionNamedRerouteDeclaration
   inputs [None {from=MaterialExpressionMultiply_74}]
   properties {name=UnderlineColor, nodeColor=<rgba 0 0.9297 1 1>}
  }
  MaterialExpressionNamedRerouteDeclaration_2 {
   class=MaterialExpressionNamedRerouteDeclaration
   inputs [None {from=MaterialExpressionCurveAtlasRowParameter_1, output=R}]
   properties {name=RedVertexColorRemap, nodeColor=<rgba 0 1 0.8125 1>}
  }
  MaterialExpressionNamedRerouteDeclaration_3 {
   class=MaterialExpressionNamedRerouteDeclaration
   inputs [None {from=MaterialExpressionMultiply_3}]
   properties {name=ToOpacity, nodeColor=<rgba 0.1953 1 0 1>}
  }
  MaterialExpressionNamedRerouteUsage_0 {class=MaterialExpressionNamedRerouteUsage}
  MaterialExpressionNamedRerouteUsage_1 {class=MaterialExpressionNamedRerouteUsage}
  MaterialExpressionNamedRerouteUsage_2 {class=MaterialExpressionNamedRerouteUsage}
  MaterialExpressionNamedRerouteUsage_20 {class=MaterialExpressionNamedRerouteUsage}
  MaterialExpressionNamedRerouteUsage_21 {class=MaterialExpressionNamedRerouteUsage}
  MaterialExpressionNamedRerouteUsage_22 {class=MaterialExpressionNamedRerouteUsage}
  MaterialExpressionNamedRerouteUsage_3 {class=MaterialExpressionNamedRerouteUsage}
  MaterialExpressionNamedRerouteUsage_4 {class=MaterialExpressionNamedRerouteUsage}
  MaterialExpressionNamedRerouteUsage_5 {class=MaterialExpressionNamedRerouteUsage}
  MaterialExpressionNamedRerouteUsage_6 {class=MaterialExpressionNamedRerouteUsage}
  MaterialExpressionOneMinus_2 {class=MaterialExpressionOneMinus, inputs [None {from=MaterialExpressionVertexColor_5, output=G}]}
  MaterialExpressionOneMinus_3 {class=MaterialExpressionOneMinus, inputs [None {from=MaterialExpressionVertexColor_37, output=G}]}
  MaterialExpressionPixelDepth_0 {class=MaterialExpressionPixelDepth, inputs [], properties {}}
  MaterialExpressionPower_6 {
   class=MaterialExpressionPower
   inputs [
    Base {from=MaterialExpressionLinearInterpolate_41}
    Exp {from=""}
   ]
  }
  MaterialExpressionPower_7 {
   class=MaterialExpressionPower
   inputs [
    Base {from=MaterialExpressionVertexColor_38, output=R}
    Exp {from=""}
   ]
   properties {constExponent=3}
  }
  MaterialExpressionSaturate_0 {class=MaterialExpressionSaturate, inputs [None {from=MaterialExpressionLinearInterpolate_13}]}
  MaterialExpressionScalarParameter_1 {class=MaterialExpressionScalarParameter, properties {defaultValue=6, parameterName=Param_22}}
  MaterialExpressionScalarParameter_10 {class=MaterialExpressionScalarParameter, properties {parameterName=Color_Blue_Min_Low}}
  MaterialExpressionScalarParameter_11 {class=MaterialExpressionScalarParameter, properties {defaultValue=0.25, parameterName=Color_Blue_Min_High}}
  MaterialExpressionScalarParameter_14 {class=MaterialExpressionScalarParameter, properties {defaultValue=4, parameterName=Param_10}}
  MaterialExpressionScalarParameter_2 {class=MaterialExpressionScalarParameter, properties {parameterName=Param_2}}
  MaterialExpressionScalarParameter_22 {class=MaterialExpressionScalarParameter, properties {defaultValue=0.8, parameterName=Param_19}}
  MaterialExpressionScalarParameter_28 {class=MaterialExpressionScalarParameter, properties {defaultValue=1, parameterName=Param_11}}
  MaterialExpressionScalarParameter_29 {class=MaterialExpressionScalarParameter, properties {defaultValue=1, parameterName=Param_12}}
  MaterialExpressionScalarParameter_3 {class=MaterialExpressionScalarParameter, properties {defaultValue=1, parameterName=Param_3}}
  MaterialExpressionScalarParameter_30 {class=MaterialExpressionScalarParameter, properties {defaultValue=0.45, parameterName=Param_20}}
  MaterialExpressionScalarParameter_31 {class=MaterialExpressionScalarParameter, properties {defaultValue=0.15, parameterName=Param_21}}
  MaterialExpressionScalarParameter_32 {class=MaterialExpressionScalarParameter, properties {defaultValue=6, parameterName=Param_22}}
  MaterialExpressionScalarParameter_33 {class=MaterialExpressionScalarParameter, properties {defaultValue=-0.25, parameterName=Param_23}}
  MaterialExpressionScalarParameter_34 {class=MaterialExpressionScalarParameter, properties {defaultValue=1, parameterName=Param_13}}
  MaterialExpressionScalarParameter_35 {class=MaterialExpressionScalarParameter, properties {parameterName=isUnderline}}
  MaterialExpressionScalarParameter_36 {class=MaterialExpressionScalarParameter, properties {defaultValue=1, parameterName=Param_14}}
  MaterialExpressionScalarParameter_37 {class=MaterialExpressionScalarParameter, properties {defaultValue=1, parameterName=Param_15}}
  MaterialExpressionScalarParameter_39 {class=MaterialExpressionScalarParameter, properties {defaultValue=1, parameterName=Param_16}}
  MaterialExpressionScalarParameter_4 {class=MaterialExpressionScalarParameter}
  MaterialExpressionScalarParameter_5 {class=MaterialExpressionScalarParameter, properties {defaultValue=1, parameterName=Color_GreenMult}}
  MaterialExpressionScalarParameter_6 {class=MaterialExpressionScalarParameter, properties {defaultValue=0.5, parameterName=Param_1}}
  MaterialExpressionSceneTexture_0 {class=MaterialExpressionSceneTexture, inputs [UVs {from="", output=""}], properties {sceneTextureId=PPI_CustomDepth}}
  MaterialExpressionShadingPathSwitch_0 {
   class=MaterialExpressionShadingPathSwitch
   inputs [
    Default {from=MaterialExpressionIf_1, output=""}
    Deferred {from="", output=""}
    Forward {from="", output=""}
    Mobile {from=MaterialExpressionConstant_12, output=""}
   ]
   properties {desc=Migrated from FeatureLevelSwitch}
  }
  MaterialExpressionSine_0 {class=MaterialExpressionSine, inputs [None {from=MaterialExpressionDivide_0}]}
  MaterialExpressionTextureSample_0 {
   class=MaterialExpressionTextureSample
   inputs [
    UVs {from=""}
    Tex {from=""}
    Apply View MipBias {from=""}
   ]
   properties {texture=&/T_FX_Contrail_RadioPaca_EmissiveEdge}
  }
  MaterialExpressionTextureSample_1 {
   class=MaterialExpressionTextureSample
   inputs [
    UVs {from=MaterialExpressionConvert_3}
    Tex {from=""}
    Apply View MipBias {from=""}
   ]
   properties {samplerType=SAMPLERTYPE_Grayscale, texture=/Game/Global/Textures/Default/Noise/LowResBlurredNoiseNormalized}
  }
  MaterialExpressionTextureSample_2 {
   class=MaterialExpressionTextureSample
   inputs [
    UVs {from=MaterialExpressionConvert_4}
    Tex {from=""}
    Apply View MipBias {from=""}
   ]
   properties {texture=/Game/Effects/Fort_Effects/Emotes/UltraLight/Texture/T_SoftNoise_BC7}
  }
  MaterialExpressionTextureSample_3 {
   class=MaterialExpressionTextureSample
   inputs [
    UVs {from=MaterialExpressionConvert_0}
    Tex {from=""}
    Apply View MipBias {from=""}
   ]
   properties {texture=&T_SoftNoise_BC7}
  }
  MaterialExpressionTextureSample_5 {
   class=MaterialExpressionTextureSample
   inputs [
    UVs {from=MaterialExpressionConvert_5}
    Tex {from=""}
    Apply View MipBias {from=""}
   ]
   properties {texture=&T_SoftNoise_BC7}
  }
  MaterialExpressionTime_0 {class=MaterialExpressionTime}
  MaterialExpressionTime_1 {class=MaterialExpressionTime}
  MaterialExpressionTime_10 {class=MaterialExpressionTime}
  MaterialExpressionTime_11 {class=MaterialExpressionTime}
  MaterialExpressionTime_2 {class=MaterialExpressionTime}
  MaterialExpressionTime_3 {class=MaterialExpressionTime}
  MaterialExpressionTime_4 {class=MaterialExpressionTime}
  MaterialExpressionTime_9 {class=MaterialExpressionTime}
  MaterialExpressionVertexColor_0 {class=MaterialExpressionVertexColor, inputs [], properties {}}
  MaterialExpressionVertexColor_1 {class=MaterialExpressionVertexColor, inputs [], properties {}}
  MaterialExpressionVertexColor_18 {class=MaterialExpressionVertexColor, inputs [], properties {}}
  MaterialExpressionVertexColor_2 {class=MaterialExpressionVertexColor, inputs [], properties {}}
  MaterialExpressionVertexColor_23 {class=MaterialExpressionVertexColor, inputs [], properties {}}
  MaterialExpressionVertexColor_28 {class=MaterialExpressionVertexColor, inputs [], properties {}}
  MaterialExpressionVertexColor_29 {class=MaterialExpressionVertexColor, inputs [], properties {}}
  MaterialExpressionVertexColor_30 {class=MaterialExpressionVertexColor, inputs [], properties {}}
  MaterialExpressionVertexColor_31 {class=MaterialExpressionVertexColor, inputs [], properties {}}
  MaterialExpressionVertexColor_32 {class=MaterialExpressionVertexColor, inputs [], properties {}}
  MaterialExpressionVertexColor_35 {class=MaterialExpressionVertexColor, inputs [], properties {}}
  MaterialExpressionVertexColor_37 {class=MaterialExpressionVertexColor, inputs [], properties {}}
  MaterialExpressionVertexColor_38 {class=MaterialExpressionVertexColor, inputs [], properties {}}
  MaterialExpressionVertexColor_39 {class=MaterialExpressionVertexColor, inputs [], properties {}}
  MaterialExpressionVertexColor_5 {class=MaterialExpressionVertexColor, inputs [], properties {}}
  MaterialExpressionVertexColor_6 {class=MaterialExpressionVertexColor, inputs [], properties {}}
 }
 properties {
  allowTranslucentCustomDepthWrites=true
  bAutomaticallySetUsageInEditor=false
  bUsedWithNiagaraMeshParticles=true
  blendMode=BLEND_Translucent
  shadingModel=MSM_Unlit
  twoSided=true
  usedShadingModels=Unlit
 }
}
