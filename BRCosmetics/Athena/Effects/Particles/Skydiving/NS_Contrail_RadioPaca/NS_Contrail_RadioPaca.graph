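# Niagara skydiving contrail for the RadioPaca cosmetic, with audio-reactive meshes and airglow sprites pinned to the character's skeletal mesh.
# Single-burst, long-lived particles form a Bottom piece plus paired HandGlow and HandTrails elements that follow sampled bones or sockets.
# Event audio amplitude is remapped through curves, smoothed in Bottom and HandTrails, then fed to dynamic material parameters alongside time-of-day and timeline scaling.
# Camera angle adjusts bottom glow alpha and hand trail rotation.
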NiagaraSystem /BRCosmetics/Athena/Effects/Particles/Skydiving/NS_Contrail_RadioPaca/NS_Contrail_RadioPaca
dependencies {
 usedDataInterfaces [
  &NiagaraDataInterfaceArrayFloat3
  &NiagaraDataInterfaceCurve
  &NiagaraDataInterfaceArrayInt32
  &NiagaraDataInterfaceArrayFloat
  &NiagaraDataInterfaceCamera
  &NiagaraDataInterfaceSkeletalMesh
 ]
 usedDynamicInputs [&SimulationPosition, &SetBoolByFloatComparison]
 usedModules [
  &self:SystemSpawnScript.SystemScriptSource.SystemScriptGraph.NiagaraNodeAssignment_4.SetVariables_#1
  &SystemState
  &Timeline
 ]
}
emitters {
 Bottom {
  emitterUpdateScript [
   &EmitterState(
    Life Cycle Mode=::NewEnumerator1
    Loop Behavior=::NewEnumerator1
    Loop Duration Mode=::NewEnumerator1
    Reset Age On Awaken=true
    Scalability Mode=::NewEnumerator1
    UseLoopDelay=false
   )
   &SpawnBurst_Instantaneous(Spawn Count=1, Use Spawn Probability=false)
  ]
  particleSpawnScript [
   &InitializeParticle(
    Lifetime=2
    Mesh Scale=&Multiply_VectorByFloat(Vector=<1 1 0.25>, Float=1.75)
    Mesh Scale Mode=::NewEnumerator4
    Sprite Size Mode=::NewEnumerator3
    Uniform Sprite Size=165
    UsePositionOffset=false
    Write Material Random=true
   ) {module=InitializeParticle001}
   /SetVariables() {isSetParameters=true}
   /SetVariables() {isSetParameters=true}
  ]
  particleUpdateScript [
   &ParticleState(Kill Particles When Lifetime Has Elapsed=false, Let Infinitely Lived Particles Die When Emitter Deactivates=true)
   /SetVariables(Particles.AmpCurve:NiagaraFloat=&FloatFromCurve(
     FloatCurve=DataInterface() {propertyValues {Curve {
        defaultValue=1
        keys [
         {interpMode=RCIM_Cubic, tangentMode=RCTM_User, time=0.3, value=0}
         {arriveTangentWeight=0.0471, interpMode=RCIM_Cubic, leaveTangentWeight=0.0471, tangentMode=RCTM_User, time=0.4, value=0.2}
         {arriveTangentWeight=0.0602, interpMode=RCIM_Cubic, leaveTangentWeight=0.0337, tangentMode=RCTM_User, time=0.5, value=0.2}
         {arriveTangentWeight=0.0601, interpMode=RCIM_Cubic, leaveTangentWeight=0.0601, tangentMode=RCTM_User, time=0.6, value=0.5}
         {arriveTangentWeight=0.0791, interpMode=RCIM_Cubic, leaveTangentWeight=0.0288, tangentMode=RCTM_User, time=0.7, value=0.5}
         {arriveTangentWeight=0.0601, interpMode=RCIM_Cubic, leaveTangentWeight=0.0601, tangentMode=RCTM_User, time=0.8, value=0.8}
         {arriveTangentWeight=0.0471, interpMode=RCIM_Cubic, leaveTangentWeight=0.0471, tangentMode=RCTM_User, time=0.9, value=0.8}
         <at 1 1 RCIM_Cubic>
        ]
       }}}
     CurveIndex=$NPC.FortniteNPC.EventAudio_Amplitude
    )) {isSetParameters=true}
   /SetVariables(Particles.AmpCurve2:NiagaraFloat=&FloatFromCurve(
     FloatCurve=DataInterface() {propertyValues {Curve {defaultValue=1, keys [{interpMode=RCIM_Cubic, tangentMode=RCTM_User, time=0.1, value=0}, <at 0.15 1 RCIM_Cubic>]}}}
     CurveIndex=&Multiply_Float(A=$NPC.FortniteNPC.EventAudio_Amplitude, B=1)
    )) {isSetParameters=true}
   &InterpolateOverTime(
    Rate Of Change:NiagaraFloat=&Lerp_Float(
     A=150
     B=1200
     Alpha=&FloatFromCurve(
      FloatCurve=DataInterface() {propertyValues {Curve {keys [<at -1 1>,<at 0 0>]}}}
      CurveIndex=&Subtract_Float(A=$Particles.AmpSmooth, B=$Particles.AmpCurve)
     )
    )
    Target Float Value:NiagaraFloat=$Particles.AmpCurve
   )
   &InterpolateOverTime(
    Rate Of Change:NiagaraFloat=&Lerp_Float(
     A=150
     B=1200
     Alpha=&FloatFromCurve(
      FloatCurve=DataInterface() {propertyValues {Curve {keys [<at -1 1>,<at 0 0>]}}}
      CurveIndex=&Subtract_Float(A=$Particles.AmpSmooth2, B=$Particles.AmpCurve2)
     )
    )
    Target Float Value:NiagaraFloat=$Particles.AmpCurve2
   ) {module=InterpolateOverTime001}
   /SetVariables(Particles.AmpSmooth:NiagaraFloat=$Particles.InterpolateOverTime.InterpolateOverTime_Float.Moving Average) {isSetParameters=true}
   /SetVariables(Particles.AmpSmooth2:NiagaraFloat=$Particles.InterpolateOverTime001.InterpolateOverTime_Float.Moving Average) {isSetParameters=true}
   &ScaleMeshSize(Scale Factor=&VectorFromFloat(Value=&Lerp_Float(A=0.7, B=1.1, Alpha=$System.Timeline.NormalizedTime)))
   &SkeletalMeshLocation(
    Bone / Socket Index=4
    Bone Sampling Mode=::NewEnumerator4
    Inherit Velocity=false
    Interpolate Mesh Sampling=false
    MaskSpawnGroup=false
    Sampled Position Offset=<0 0 -5>
    Sampled Position Offset Coordinate Space:ENiagaraMeshTransforms=::NewEnumerator3
    Skeletal Mesh=$System.Skeletal Mesh
   )
   /SetVariables(Particles.Audio:NiagaraFloat=&ClampFloat(Float=&Multiply_Float(
      A=&Power(
       a=&FloatFromCurve(
        FloatCurve=DataInterface() {propertyValues {Curve {
           defaultValue=1
           keys [
            {interpMode=RCIM_Cubic, tangentMode=RCTM_User, time=0.3, value=0}
            {arriveTangentWeight=0.0471, interpMode=RCIM_Cubic, leaveTangentWeight=0.0471, tangentMode=RCTM_User, time=0.4, value=0.2}
            {arriveTangentWeight=0.0602, interpMode=RCIM_Cubic, leaveTangentWeight=0.0337, tangentMode=RCTM_User, time=0.5, value=0.2}
            {arriveTangentWeight=0.0601, interpMode=RCIM_Cubic, leaveTangentWeight=0.0601, tangentMode=RCTM_User, time=0.6, value=0.5}
            {arriveTangentWeight=0.0791, interpMode=RCIM_Cubic, leaveTangentWeight=0.0288, tangentMode=RCTM_User, time=0.7, value=0.5}
            {arriveTangentWeight=0.0601, interpMode=RCIM_Cubic, leaveTangentWeight=0.0601, tangentMode=RCTM_User, time=0.8, value=0.8}
            {arriveTangentWeight=0.0471, interpMode=RCIM_Cubic, leaveTangentWeight=0.0471, tangentMode=RCTM_User, time=0.9, value=0.8}
            <at 1 1 RCIM_Cubic>
           ]
          }}}
        CurveIndex=$NPC.FortniteNPC.EventAudio_Amplitude
       )
       b=2
      )
      B=1.5
     ))) {enabled=false, isSetParameters=true}
   /SetVariables(Particles.Audio:NiagaraFloat=$Particles.AmpSmooth) {isSetParameters=true}
   /SetVariables(Particles.Audio2:NiagaraFloat=$Particles.AmpSmooth2) {isSetParameters=true}
   &DynamicMaterialParameters(
    Index 0 Param 1=$Particles.Audio
    Index 0 Param 2=&Lerp_Float(
     A=0.7
     B=&FNTimeOfDayScalar(Night Scale=0.85, Morning Scale=1.2, Day Scale=1.2, Evening Scale=1.2)
     Alpha=$System.Timeline.NormalizedTime
    )
    Index 0 Param 3=&Lerp_Float(
     A=0.7
     B=&FNTimeOfDayScalar(Night Scale=0.9, Morning Scale=1, Evening Scale=1)
     Alpha=$System.Timeline.NormalizedTime
    )
    Index 0 Param 4=$Particles.Audio2
    Module.Index0Float0Write=true
    Module.Index0Float1Write=true
    Module.Index0Float2Write=true
    Module.Index0Float3Write=true
   )
   /SetVariables() {enabled=false, isSetParameters=true}
   /SetVariables(Particles.AirglowColor:LinearColor=&MultiplyLinearColor(
     Linear Color=<rgba 0.0152 0.5972 0.9647 0.25>
     Scale Factor=&MakeVector4(
      X=1
      Y=1
      Z=1
      W=&Multiply_Float(
       A=1
       B=&FloatFromCurve(
        FloatCurve=DataInterface() {propertyValues {Curve {keys [<at -1 1 RCIM_Cubic>,<at -0.25 0 RCIM_Cubic>,<at 1 1 RCIM_Cubic>]}}}
        CurveIndex=&DotProduct(Vector A=<0 0 -1>, Vector B=&ReturnCameraVector(Camera Vectors=::NewEnumerator1, Coordinate Space=::World))
        Scale Curve=&Lerp_Float(A=1, B=1.5, Alpha=$System.Timeline.NormalizedTime)
       )
      )
     )
    )) {isSetParameters=true}
   /SetVariables(Particles.dynairglow:Vector4f=&MakeVector4(X=1.5, Y=20, Z=1, W=1)) {isSetParameters=true}
   &SpriteFacingAndAlignment(Sprite Facing=<0 0 1>, WriteAlignment=false, WriteFacing=true)
  ]
  rendererClasses [&NiagaraMeshRendererProperties, &NiagaraSpriteRendererProperties]
  renderers [
   {rendererClass=&NiagaraMeshRendererProperties, rendererIndex=0}
   {rendererClass=&NiagaraSpriteRendererProperties, rendererIndex=1}
  ]
 }
 HandGlow {
  emitterUpdateScript [
   &EmitterState(Life Cycle Mode=::NewEnumerator1, Loop Behavior=::NewEnumerator1, Loop Duration Mode=::NewEnumerator1, UseLoopDelay=false)
   &SpawnBurst_Instantaneous(Spawn Count=2, Use Spawn Probability=false, UseLoopCountLimit=false)
  ]
  particleSpawnScript [
   &PartitionParticles()
   &InitializeParticle(
    Color Mode=::NewEnumerator0
    Mesh Scale Mode=::NewEnumerator3
    Mesh Uniform Scale=0.1
    Sprite Size Mode=::NewEnumerator3
    Uniform Sprite Size=75
    UsePositionOffset=false
    Write Material Random=true
   )
   &InitialMeshOrientation(
    Mesh Orientation Mode=::NewEnumerator7
    Rotation=&SelectVectorFromArray(
     Direct Array Index=$Particles.PartitionParticles.Partition
     Vector Selection Array=DataInterface() {propertyValues {FloatData [<0 0 0.5>, <0 0 0>, <0 0 0>, <0 0 0>]}}
    )
    Use Rotation Vector=true
   )
  ]
  particleUpdateScript [
   &ParticleState(
    Kill Particles When Lifetime Has Elapsed=false
    Let Infinitely Lived Particles Die When Emitter Deactivates=true
    Loop Particles Lifetime=true
   )
   &ScaleMeshSize(Scale Factor=&VectorFromFloat(Value=5))
   &SkeletalMeshLocation(
    Bone / Socket Index=&SelectIntFromArray(Direct Array Index=$Particles.PartitionParticles.Partition, Int Selection Array=DataInterface() {propertyValues {IntData [0,1,0,1]}})
    Bone Sampling Mode=::NewEnumerator4
    MaskSpawnGroup=false
    Orientation Sampling=::NewEnumerator1
    Sampled Position Offset=&SelectVectorFromArray(
     Direct Array Index=$Particles.PartitionParticles.Partition
     Vector Selection Array=DataInterface() {propertyValues {FloatData [<5 0 10>, <0 0 -10>]}}
    )
    Skeletal Mesh=$System.Skeletal Mesh
   )
   /SetVariables(Particles.Audio:NiagaraFloat=&ClampFloat(Float=&Multiply_Float(
      A=&Power(
       a=&FloatFromCurve(
        FloatCurve=DataInterface() {propertyValues {Curve {
           defaultValue=1
           keys [{interpMode=RCIM_Cubic, tangentMode=RCTM_User, time=0.3, value=0}, {interpMode=RCIM_Cubic, tangentMode=RCTM_User, time=1, value=1}]
          }}}
        CurveIndex=$NPC.FortniteNPC.EventAudio_Amplitude
       )
       b=2
      )
      B=1.5
     ))) {isSetParameters=true, module=SetVariables_#2}
   &DynamicMaterialParameters(
    Index 0 Param 1=$Particles.Audio
    Module.Index0Float0Write=true
    Module.Index0Float1Write=false
    Module.Index0Float2Write=false
    Module.Index0Float3Write=false
   )
   /SetVariables(Particles.dynairglow:Vector4f=&MakeVector4(X=2, Y=20, Z=1, W=1)) {isSetParameters=true, module=SetVariables_#3}
   /SetVariables(Particles.AirglowColor:LinearColor=<rgba 0.0152 0.5972 0.9647 0.2>) {isSetParameters=true, module=SetVariables_#4}
   /SetVariables(Particles.AirglowPosition:NiagaraPosition=&AddVectorToPosition(
     Position=$Particles.Position
     Vector=&RotateVectorByQuaternion(
      Vector To Rotate=&SelectVectorFromArray(
       Direct Array Index=$Particles.PartitionParticles.Partition
       Vector Selection Array=DataInterface() {propertyValues {FloatData [<0 0 -10>, <0 0 10>]}}
      )
      Quaternion=$Particles.SkeletalMeshLocation.SampledRotation
     )
    )) {isSetParameters=true, module=SetVariables_#5}
  ]
  rendererClasses [&NiagaraMeshRendererProperties, &NiagaraSpriteRendererProperties]
  renderers [
   {rendererClass=&NiagaraMeshRendererProperties, rendererIndex=0}
   {rendererClass=&NiagaraSpriteRendererProperties, rendererIndex=1}
  ]
 }
 HandTrails {
  emitterUpdateScript [
   &EmitterState(Life Cycle Mode=::NewEnumerator1, Loop Behavior=::NewEnumerator1, Loop Duration Mode=::NewEnumerator1, UseLoopDelay=false)
   &SpawnBurst_Instantaneous(Spawn Count=2, Use Spawn Probability=false, UseLoopCountLimit=false)
  ]
  particleSpawnScript [
   &PartitionParticles()
   &InitializeParticle(Color Mode=::NewEnumerator0, Mesh Scale Mode=::NewEnumerator3, Mesh Uniform Scale=0.1, UsePositionOffset=false, Write Material Random=true)
   &InitialMeshOrientation(
    Mesh Orientation Mode=::NewEnumerator7
    Rotation=&SelectVectorFromArray(
     Direct Array Index=$Particles.PartitionParticles.Partition
     Vector Selection Array=DataInterface() {propertyValues {FloatData [<0 0 0.5>, <0 0 0>, <0 0 0>, <0 0 0>]}}
    )
    Use Rotation Vector=true
   )
   /SetVariables() {isSetParameters=true}
   /SetVariables() {isSetParameters=true}
  ]
  particleUpdateScript [
   &ParticleState(
    Kill Particles When Lifetime Has Elapsed=false
    Let Infinitely Lived Particles Die When Emitter Deactivates=true
    Loop Particles Lifetime=true
   )
   /SetVariables(Particles.AmpCurve:NiagaraFloat=&FloatFromCurve(
     FloatCurve=DataInterface() {propertyValues {Curve {
        defaultValue=1
        keys [
         {interpMode=RCIM_Cubic, tangentMode=RCTM_User, time=0.3, value=0}
         {arriveTangentWeight=0.0471, interpMode=RCIM_Cubic, leaveTangentWeight=0.0471, tangentMode=RCTM_User, time=0.4, value=0.2}
         {arriveTangentWeight=0.0602, interpMode=RCIM_Cubic, leaveTangentWeight=0.0337, tangentMode=RCTM_User, time=0.5, value=0.2}
         {arriveTangentWeight=0.0601, interpMode=RCIM_Cubic, leaveTangentWeight=0.0601, tangentMode=RCTM_User, time=0.6, value=0.5}
         {arriveTangentWeight=0.0791, interpMode=RCIM_Cubic, leaveTangentWeight=0.0288, tangentMode=RCTM_User, time=0.7, value=0.5}
         {arriveTangentWeight=0.0601, interpMode=RCIM_Cubic, leaveTangentWeight=0.0601, tangentMode=RCTM_User, time=0.8, value=0.8}
         {arriveTangentWeight=0.0471, interpMode=RCIM_Cubic, leaveTangentWeight=0.0471, tangentMode=RCTM_User, time=0.9, value=0.8}
         <at 1 1 RCIM_Cubic>
        ]
       }}}
     CurveIndex=$NPC.FortniteNPC.EventAudio_Amplitude
    )) {isSetParameters=true}
   /SetVariables(Particles.AmpCurve2:NiagaraFloat=&FloatFromCurve(
     FloatCurve=DataInterface() {propertyValues {Curve {defaultValue=1, keys [{interpMode=RCIM_Cubic, tangentMode=RCTM_User, time=0.1, value=0}, <at 0.15 1 RCIM_Cubic>]}}}
     CurveIndex=&Multiply_Float(A=$NPC.FortniteNPC.EventAudio_Amplitude, B=1)
    )) {isSetParameters=true}
   &InterpolateOverTime(
    Rate Of Change:NiagaraFloat=&Lerp_Float(
     A=150
     B=1200
     Alpha=&FloatFromCurve(
      FloatCurve=DataInterface() {propertyValues {Curve {keys [<at -1 1>,<at 0 0>]}}}
      CurveIndex=&Subtract_Float(A=$Particles.AmpSmooth, B=$Particles.AmpCurve)
     )
    )
    Target Float Value:NiagaraFloat=$Particles.AmpCurve
   )
   &InterpolateOverTime(
    Rate Of Change:NiagaraFloat=&Lerp_Float(
     A=150
     B=1200
     Alpha=&FloatFromCurve(
      FloatCurve=DataInterface() {propertyValues {Curve {keys [<at -1 1>,<at 0 0>]}}}
      CurveIndex=&Subtract_Float(A=$Particles.AmpSmooth2, B=$Particles.AmpCurve2)
     )
    )
    Target Float Value:NiagaraFloat=$Particles.AmpCurve2
   ) {module=InterpolateOverTime001}
   /SetVariables(Particles.AmpSmooth:NiagaraFloat=$Particles.InterpolateOverTime.InterpolateOverTime_Float.Moving Average) {isSetParameters=true}
   /SetVariables(Particles.AmpSmooth2:NiagaraFloat=$Particles.InterpolateOverTime001.InterpolateOverTime_Float.Moving Average) {isSetParameters=true}
   &ScaleMeshSize(Scale Factor=&VectorFromFloat(Value=5))
   &SkeletalMeshLocation(
    Bone / Socket Index=&SelectIntFromArray(Direct Array Index=$Particles.PartitionParticles.Partition, Int Selection Array=DataInterface() {propertyValues {IntData [0,1,0,1]}})
    Bone Sampling Mode=::NewEnumerator4
    MaskSpawnGroup=false
    Orientation Sampling=::NewEnumerator1
    Sampled Position Offset=&SelectVectorFromArray(
     Direct Array Index=$Particles.PartitionParticles.Partition
     Vector Selection Array=DataInterface() {propertyValues {FloatData [<-10 -5 0>, <-10 -5 0>, <0 0 0>, <0 0 0>]}}
    )
    Sampled Position Offset Coordinate Space:ENiagaraSkelMeshTransforms=::NewEnumerator4
    Skeletal Mesh=$System.Skeletal Mesh
   )
   &UpdateMeshOrientation(
    Facing Direction=<0.2 0.1 0.2>
    Facing Interpolation=false
    Orientation Method=::NewEnumerator1
    Override Quat=false
    TransformFacing=false
    TransformUp=false
    Up Direction:Vector3f=<0 1 0>
   ) {enabled=false}
   /SetVariables(Particles.Audio:NiagaraFloat=&ClampFloat(Float=&Multiply_Float(
      A=&Power(
       a=&FloatFromCurve(
        FloatCurve=DataInterface() {propertyValues {Curve {
           defaultValue=1
           keys [
            {interpMode=RCIM_Cubic, tangentMode=RCTM_User, time=0.3, value=0}
            {arriveTangentWeight=0.0471, interpMode=RCIM_Cubic, leaveTangentWeight=0.0471, tangentMode=RCTM_User, time=0.4, value=0.2}
            {arriveTangentWeight=0.0602, interpMode=RCIM_Cubic, leaveTangentWeight=0.0337, tangentMode=RCTM_User, time=0.5, value=0.2}
            {arriveTangentWeight=0.0601, interpMode=RCIM_Cubic, leaveTangentWeight=0.0601, tangentMode=RCTM_User, time=0.6, value=0.5}
            {arriveTangentWeight=0.0791, interpMode=RCIM_Cubic, leaveTangentWeight=0.0288, tangentMode=RCTM_User, time=0.7, value=0.5}
            {arriveTangentWeight=0.0601, interpMode=RCIM_Cubic, leaveTangentWeight=0.0601, tangentMode=RCTM_User, time=0.8, value=0.8}
            {arriveTangentWeight=0.0471, interpMode=RCIM_Cubic, leaveTangentWeight=0.0471, tangentMode=RCTM_User, time=0.9, value=0.8}
            <at 1 1 RCIM_Cubic>
           ]
          }}}
        CurveIndex=$NPC.FortniteNPC.EventAudio_Amplitude
       )
       b=2
      )
      B=1.5
     ))) {enabled=false, isSetParameters=true, module=SetVariables_#2}
   /SetVariables(Particles.Audio:NiagaraFloat=$Particles.AmpSmooth) {isSetParameters=true}
   /SetVariables(Particles.Audio2:NiagaraFloat=$Particles.AmpSmooth2) {isSetParameters=true}
   &DynamicMaterialParameters(
    Index 0 Param 1=$Particles.Audio
    Index 0 Param 2=&Lerp_Float(
     A=0.7
     B=&FNTimeOfDayScalar(Night Scale=0.85, Morning Scale=1, Evening Scale=1)
     Alpha=$System.Timeline.NormalizedTime
    )
    Index 0 Param 3=&Lerp_Float(
     A=0.7
     B=&FNTimeOfDayScalar(Night Scale=0.8, Morning Scale=0.9, Day Scale=0.9, Evening Scale=0.9)
     Alpha=$System.Timeline.NormalizedTime
    )
    Index 0 Param 4=$Particles.Audio2
    Module.Index0Float0Write=true
    Module.Index0Float1Write=true
    Module.Index0Float2Write=true
    Module.Index0Float3Write=true
   )
   /SetVariables(Particles.dynairglow:Vector4f=&MakeVector4(X=2, Y=20, Z=1, W=1)) {enabled=false, isSetParameters=true, module=SetVariables_#3}
   /SetVariables(Particles.AirglowColor:LinearColor=<rgba 0.0152 0.5972 0.9647 1>) {enabled=false, isSetParameters=true, module=SetVariables_#4}
   /SetVariables(Particles.AirglowPosition:NiagaraPosition=&AddVectorToPosition(
     Position=$Particles.Position
     Vector=&RotateVectorByQuaternion(
      Vector To Rotate=&SelectVectorFromArray(
       Direct Array Index=$Particles.PartitionParticles.Partition
       Vector Selection Array=DataInterface() {propertyValues {FloatData [<0 0 -10>, <0 0 10>]}}
      )
      Quaternion=$Particles.SkeletalMeshLocation.SampledRotation
     )
    )) {enabled=false, isSetParameters=true, module=SetVariables_#5}
   /SetVariables(Particles.MeshOrientation=&MultiplyQuaternion(
     Quaternion A=$Particles.Initial.MeshOrientation
     Quaternion B=&MakeQuaternion(
      Quaternion From=0
      Axis=<1 0 0>
      Angle=&Lerp_Float(
       A=0
       B=&SelectFloatFromArray(Direct Array Index=$Particles.PartitionParticles.Partition, Float Selection Array=DataInterface() {propertyValues {FloatData [-12,12]}})
       Alpha=&FloatFromCurve(
        FloatCurve=DataInterface() {propertyValues {Curve {keys [<at 0.8 0>,<at 1 1>]}}}
        CurveIndex=&DotProduct(
         Vector A=&ReturnCameraVector(Camera Vectors=::NewEnumerator1, Coordinate Space=::Simulation)
         Vector B=<0 0 -1>
        )
       )
      )
      From=?
      To=?
      Angle Type=::NewEnumerator0
      Normalize Vectors=false
     )
    )) {isSetParameters=true}
  ]
  rendererClasses [&NiagaraMeshRendererProperties]
  renderers [
   {rendererClass=&NiagaraMeshRendererProperties, rendererIndex=0}
   {rendererClass=&NiagaraMeshRendererProperties, rendererIndex=1}
  ]
 }
}
summary {
 systemName=NS_Contrail_RadioPaca
 emitters [
  {emitterName=HandTrails, rendererClasses [&NiagaraMeshRendererProperties]}
  {emitterName=Bottom, rendererClasses [&NiagaraMeshRendererProperties, &NiagaraSpriteRendererProperties]}
  {emitterName=HandGlow, rendererClasses [&NiagaraMeshRendererProperties, &NiagaraSpriteRendererProperties]}
 ]
 userVariables [
  User.bIsFrontEnd {type=NiagaraBool}
  User.bIsFrontEndPreview {type=NiagaraBool}
  User.ParaGlideLeanAlpha {type=NiagaraFloat, defaultValue=0, description=""}
  User.RibbonIDOffset {type=NiagaraInt32}
 ]
}
